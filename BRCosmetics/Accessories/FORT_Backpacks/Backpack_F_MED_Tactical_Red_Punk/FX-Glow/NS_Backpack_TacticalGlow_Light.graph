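# Niagara system for the Tactical Red Punk backpack that adds a yellow-green blade glow and light.
# Two burst spawn groups index into skeletal mesh bones or sockets, pinning sprites there and aligning them to the sampled rotation.
# A curve on emitter age fades sprite alpha in, while an offset position and light attributes feed a light renderer.
# User.ShowBlade switches glow and light in gameplay, with front-end modifiers supplying full values in menus.
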NiagaraSystem /BRCosmetics/Accessories/FORT_Backpacks/Backpack_F_MED_Tactical_Red_Punk/FX-Glow/NS_Backpack_TacticalGlow_Light
dependencies {usedDataInterfaces [&NiagaraDataInterfaceCurve], usedDynamicInputs [], usedModules [&SystemState]}
emitters {Blade_Glow_and_Light {
  emitterUpdateScript [
   &EmitterState(
    Inactive Response=::NewEnumerator1
    Life Cycle Mode=::NewEnumerator1
    Loop Behavior=::NewEnumerator1
    Loop Duration Mode=::NewEnumerator1
    UseLoopDelay=false
   )
   &SpawnBurst_Instantaneous(Use Spawn Probability=false, UseLoopCountLimit=false)
   &SpawnBurst_Instantaneous(Spawn Group=1, Use Spawn Probability=false, UseLoopCountLimit=false) {module=SpawnBurst_Instantaneous001}
  ]
  particleSpawnScript [&InitializeParticle(Color=<rgba 0.4458 1 0 0.1>, Sprite Size=<12 30>, Sprite Size Mode=::NewEnumerator4, UsePositionOffset=false, Write Material Random=true)]
  particleUpdateScript [
   &SkeletalMeshLocation(
    Bone / Socket Index=&SpawnGroup()
    MaskSpawnGroup=false
    Mesh Sampling Type=::NewEnumerator5
    Sampled Position Offset=<0 0 2>
    Skeletal Mesh=$User.SkeletalMEsh
    Socket Sampling Mode=::NewEnumerator5
   )
   &SpriteFacingAndAlignment(
    Sprite Alignment=&RotateVectorByQuaternion(Vector To Rotate=<0 0 1>, Quaternion=$Particles.SkeletalMeshLocation.SampledRotation)
    Sprite Facing=<0 0 0>
    WriteAlignment=true
    WriteFacing=false
   )
   &ParticleState(Kill Particles When Lifetime Has Elapsed=false, Loop Particles Lifetime=true)
   &ScaleColor(
    Scale Alpha=&FloatFromCurve(
     FloatCurve=DataInterface() {propertyValues {Curve {keys [<at 0 0>,<at 0.25 1>]}}}
     CurveIndex=$Emitter.Age
     Scale Curve=&FN_FrontEndFloatModifier(Gameplay=&MakeCustomFloatFromBool(Bool=$User.ShowBlade, True Float=1, False Float=0), Front End=1)
    )
    ScaleA=true
    ScaleRGB=true
   )
   /SetVariables(Particles.LightPosition:NiagaraPosition=&Add_Vector(
     A=Linked() {linkedVariable {name=Particles.Position, type=NiagaraPosition}}
     B=&TransformVector(Vector=<0 -10 0>, Destination Space=::Simulation)
    )) {isSetParameters=true}
   &FN_Light_Attributes(
    Color=<rgba 0.4458 1 0 0.75>
    Create Color Attribute=true
    Create Exponent Attribute=true
    Create Radius Attribute=true
    Enable Particle Light Enable=true
    Exponent=3
    Light Enabled=true
    Radius=&FN_FrontEndFloatModifier(
     Gameplay=&MakeCustomFloatFromBool(Bool=$User.ShowBlade, True Float=27.5, False Float=0)
     Front End=27.5
    )
    Scale Color=&FN_FrontEndFloatModifier(Gameplay=&MakeCustomFloatFromBool(Bool=$User.ShowBlade, True Float=1, False Float=0), Front End=1)
   )
  ]
  rendererClasses [&NiagaraSpriteRendererProperties, &NiagaraLightRendererProperties]
  renderers [
   {rendererClass=&NiagaraSpriteRendererProperties}
   {rendererClass=&NiagaraLightRendererProperties, rendererIndex=1}
   {rendererClass=&NiagaraSpriteRendererProperties, rendererIndex=2}
  ]
 }}
summary {
 systemName=NS_Backpack_TacticalGlow_Light
 emitters [{
   emitterName=Blade_Glow_and_Light
   rendererClasses [&NiagaraSpriteRendererProperties, &NiagaraLightRendererProperties]
  }]
 userVariables [
  User.bIsFrontEnd {type=NiagaraBool}
  User.bIsFrontEndPreview {type=NiagaraBool, defaultValue=true}
  User.ShowBlade {type=NiagaraBool}
  User.SkeletalMEsh {
   type=NiagaraDataInterfaceSkeletalMesh
   defaultValue {
    struct=NiagaraExt_VariableValue_DataInterface
    value {dataInterface=&self:NiagaraDataInterfaceSkeletalMesh_0}
   }
  }
 ]
}
